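# Sound cue for the Deimos Brute monster landing on dirt, built from reused human player run footstep waves.
# A Mixer layers two Random nodes, one picking a generic run step and the other a dirt run step, then feeds a Modulator at the root for variation.
# It uses Deimos step attenuation and global and local step concurrency, plays in the player/enemy movement sound class at reduced volume, and has virtualization disabled.
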SoundCue /Game/Sounds/Fort_Monster_Deimos/SFX/Brute_Land_Dirt_Cue
root=Modulator_3
nodes [
 {id=Mixer_0, children [Random_5,Random_0], properties {inputVolume [0.8,0.7]}}
 {id=Modulator_3, children [Mixer_0]}
 {
  id=Random_0
  children [WavePlayer_1,WavePlayer_2,WavePlayer_3,WavePlayer_4,WavePlayer_5]
  properties {bShouldExcludeFromBranchCulling=true, weights [1,1,1,1,1]}
 }
 {
  id=Random_5
  children [WavePlayer_1900, WavePlayer_1902, WavePlayer_1903, WavePlayer_1904]
  properties {bShouldExcludeFromBranchCulling=true, weights [1,1,1,1]}
 }
 {id=WavePlayer_1, properties {soundWaveAssetPtr=/Game/Sounds/Fort_Step_Sounds/Human/player_step_run_dirt_01}}
 {id=WavePlayer_1900, properties {soundWaveAssetPtr=/Game/Sounds/Fort_Step_Sounds/Human/player_run_01}}
 {id=WavePlayer_1902, properties {soundWaveAssetPtr=/Game/Sounds/Fort_Step_Sounds/Human/player_run_03}}
 {id=WavePlayer_1903, properties {soundWaveAssetPtr=/Game/Sounds/Fort_Step_Sounds/Human/player_run_04}}
 {id=WavePlayer_1904, properties {soundWaveAssetPtr=/Game/Sounds/Fort_Step_Sounds/Human/player_run_05}}
 {id=WavePlayer_2, properties {soundWaveAssetPtr=/Game/Sounds/Fort_Step_Sounds/Human/player_step_run_dirt_02}}
 {id=WavePlayer_3, properties {soundWaveAssetPtr=/Game/Sounds/Fort_Step_Sounds/Human/player_step_run_dirt_04}}
 {id=WavePlayer_4, properties {soundWaveAssetPtr=/Game/Sounds/Fort_Step_Sounds/Human/player_step_run_dirt_05}}
 {id=WavePlayer_5, properties {soundWaveAssetPtr=/Game/Sounds/Fort_Step_Sounds/Human/player_step_run_dirt_07}}
]
properties {
 attenuationSettings=/Game/Sounds/Attenuation/Fort_Deimos_Step_Att
 maxDistance=2000
 priority=0
 soundClassObject=/Game/Sounds/FortSoundClassesAndModes/Fort_09_Player_Enemy_Movement
 virtualizationMode=Disabled
 volumeMultiplier=0.35
 attenuationOverrides {falloffDistance=1700, attenuationShapeExtents {x=300}}
 concurrencyOverrides {resolutionRule=PreventNew}
 concurrencySet [/Game/Sounds/Fort_Monster_Deimos/Settings/Deimos_Step_Global_Conc, /Game/Sounds/Fort_Monster_Deimos/Settings/Deimos_Step_Local_Conc]
}
