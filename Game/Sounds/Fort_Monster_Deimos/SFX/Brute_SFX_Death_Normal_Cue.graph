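# Sound cue for the Deimos Brute's normal death effect, currently silent because it has no root or sound nodes.
# It carries only playback settings: half volume, Deimos close vocal attenuation, and the enemy-damaged sound class.
# A shared Deimos global concurrency set is assigned alongside concurrency override values, and virtualization is disabled.
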SoundCue /Game/Sounds/Fort_Monster_Deimos/SFX/Brute_SFX_Death_Normal_Cue
root=null
nodes []
properties {
 attenuationSettings=/Game/Sounds/Attenuation/Fort_Deimos_Vocal_Close_Att
 maxDistance=4000
 soundClassObject=/Game/Sounds/FortSoundClassesAndModes/Fort_Enemy_Damaged_PSM
 virtualizationMode=Disabled
 volumeMultiplier=0.5
 concurrencyOverrides {maxCount=2}
 concurrencySet [/Game/Sounds/Fort_Monster_Deimos/Settings/Deimos_Global_Conc]
}
