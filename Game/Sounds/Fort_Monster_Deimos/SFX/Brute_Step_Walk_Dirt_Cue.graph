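# Sound cue for the Deimos Brute monster's walking footsteps on dirt, built from the Brute's run-step recordings.
# A Random node picks among nine equally weighted waves and is excluded from branch culling, feeding a Modulator that varies pitch.
# It uses Deimos step attenuation, global and local step concurrency sets, and the player/enemy movement sound class.
# It has zero priority and virtualization disabled.
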SoundCue /Game/Sounds/Fort_Monster_Deimos/SFX/Brute_Step_Walk_Dirt_Cue
root=Modulator_2
nodes [
 {id=Modulator_2, children [Random_0], properties {pitchMax=1.5, volumeMax=1}}
 {
  id=Random_0
  children [WavePlayer_0,WavePlayer_1,WavePlayer_2,WavePlayer_3,WavePlayer_4,WavePlayer_5,WavePlayer_6,WavePlayer_7,WavePlayer_8]
  properties {bShouldExcludeFromBranchCulling=true, weights [1,1,1,1,1,1,1,1,1]}
 }
 {id=WavePlayer_0, properties {soundWaveAssetPtr=&/Brute_Step_Run_01}}
 {id=WavePlayer_1, properties {soundWaveAssetPtr=&/Brute_Step_Run_02}}
 {id=WavePlayer_2, properties {soundWaveAssetPtr=&/Brute_Step_Run_03}}
 {id=WavePlayer_3, properties {soundWaveAssetPtr=&/Brute_Step_Run_04}}
 {id=WavePlayer_4, properties {soundWaveAssetPtr=&/Brute_Step_Run_05}}
 {id=WavePlayer_5, properties {soundWaveAssetPtr=&/Brute_Step_Run_06}}
 {id=WavePlayer_6, properties {soundWaveAssetPtr=&/Brute_Step_Run_07}}
 {id=WavePlayer_7, properties {soundWaveAssetPtr=&/Brute_Step_Run_08}}
 {id=WavePlayer_8, properties {soundWaveAssetPtr=&/Brute_Step_Run_09}}
]
properties {
 attenuationSettings=/Game/Sounds/Attenuation/Fort_Deimos_Step_Att
 bBypassVolumeScaleForPriority=true
 bDebug=true
 maxDistance=2000
 priority=0
 soundClassObject=/Game/Sounds/FortSoundClassesAndModes/Fort_09_Player_Enemy_Movement
 virtualizationMode=Disabled
 volumeMultiplier=1.77
 concurrencyOverrides {maxCount=8, resolutionRule=StopOldest}
 concurrencySet [/Game/Sounds/Fort_Monster_Deimos/Settings/Deimos_Step_Global_Conc, /Game/Sounds/Fort_Monster_Deimos/Settings/Deimos_Step_Local_Conc]
}
