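# Sound cue for the 2D hit-impact sound of the Deimos Fiend monster, routed to the enemy-damaged sound class.
# A quality-level node sends the first two levels to an equal-weight random pick of five impact waves and the last two to a pick of three.
# Virtualization is disabled, and the cue references a Deimos hit-notification concurrency set alongside an enabled stop-oldest concurrency override.
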SoundCue /Game/Sounds/Fort_Monster_Deimos/SFX/Deimos_Fiend_HitImpact2D_Cue
root=QualityLevel_0
nodes [
 {id=QualityLevel_0, children [Random_0,Random_0,Random_1,Random_1]}
 {
  id=Random_0
  children [WavePlayer_1,WavePlayer_2,WavePlayer_3,WavePlayer_4,WavePlayer_5]
  properties {bShouldExcludeFromBranchCulling=true, weights [1,1,1,1,1]}
 }
 {id=Random_1, children [WavePlayer_0,WavePlayer_6,WavePlayer_7], properties {bShouldExcludeFromBranchCulling=true, weights [1,1,1]}}
 {id=WavePlayer_0, properties {soundWaveAssetPtr=&/Deimos_Fiend_HitImpact2D_05}}
 {id=WavePlayer_1, properties {soundWaveAssetPtr=&Deimos_Fiend_HitImpact2D_05}}
 {id=WavePlayer_2, properties {soundWaveAssetPtr=&/Deimos_Fiend_HitImpact2D_01}}
 {id=WavePlayer_3, properties {soundWaveAssetPtr=&/Deimos_Fiend_HitImpact2D_02}}
 {id=WavePlayer_4, properties {soundWaveAssetPtr=&/Deimos_Fiend_HitImpact2D_03}}
 {id=WavePlayer_5, properties {soundWaveAssetPtr=&/Deimos_Fiend_HitImpact2D_04}}
 {id=WavePlayer_6, properties {soundWaveAssetPtr=&Deimos_Fiend_HitImpact2D_01}}
 {id=WavePlayer_7, properties {soundWaveAssetPtr=&Deimos_Fiend_HitImpact2D_02}}
]
properties {
 bOverrideConcurrency=true
 maxDistance=2097152
 priority=0.5
 soundClassObject=/Game/Sounds/FortSoundClassesAndModes/Fort_Enemy_Damaged_PSM
 virtualizationMode=Disabled
 volumeMultiplier=0.7
 concurrencyOverrides {maxCount=12, resolutionRule=StopOldest}
 concurrencySet [/Game/Sounds/Fort_Monster_Deimos/Settings/Deimos_HItNotifications_Conc]
}
