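# Sound cue for the 2D hit impact notification of the Deimos Ice Fiend monster.
# A short fixed delay feeds a Random node that picks one of five equally weighted impact waves and is excluded from branch culling.
# It plays on the enemy-damaged sound class at slightly reduced volume, with virtualization disabled and a shared hit-notification concurrency setting limiting overlap.
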SoundCue /Game/Sounds/Fort_Monster_Deimos/SFX/Deimos_Fiend_Ice_HitImpact2D_Cue
root=Delay_0
nodes [
 {id=Delay_0, children [Random_1], properties {delayMax=0.15, delayMin=0.15}}
 {
  id=Random_1
  children [WavePlayer_0,WavePlayer_6,WavePlayer_7,WavePlayer_8,WavePlayer_9]
  properties {bShouldExcludeFromBranchCulling=true, weights [1,1,1,1,1]}
 }
 {id=WavePlayer_0, properties {soundWaveAssetPtr=&/Deimos_FiendIce_HitImpact2D_01}}
 {id=WavePlayer_6, properties {soundWaveAssetPtr=&/Deimos_FiendIce_HitImpact2D_02}}
 {id=WavePlayer_7, properties {soundWaveAssetPtr=&/Deimos_FiendIce_HitImpact2D_03}}
 {id=WavePlayer_8, properties {soundWaveAssetPtr=&/Deimos_FiendIce_HitImpact2D_04}}
 {id=WavePlayer_9, properties {soundWaveAssetPtr=&/Deimos_FiendIce_HitImpact2D_05}}
]
properties {
 maxDistance=2097152
 soundClassObject=/Game/Sounds/FortSoundClassesAndModes/Fort_Enemy_Damaged_PSM
 virtualizationMode=Disabled
 volumeMultiplier=0.8
 concurrencySet [/Game/Sounds/Fort_Monster_Deimos/Settings/Deimos_HItNotifications_Conc]
}
