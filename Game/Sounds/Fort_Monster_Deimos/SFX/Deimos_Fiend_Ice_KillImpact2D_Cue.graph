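# Sound cue for the 2D kill impact on a Deimos Ice Fiend, routed to the enemy-damaged sound class with hit-notification concurrency.
# A short fixed delay feeds a quality-level switch.
# The first two quality slots mix a random Ice Fiend hit impact with a random EvilCowboy pickaxe impact.
# The last two slots play only a random standard Fiend hit impact, without either layer.
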SoundCue /Game/Sounds/Fort_Monster_Deimos/SFX/Deimos_Fiend_Ice_KillImpact2D_Cue
root=Delay_0
nodes [
 {id=Delay_0, children [QualityLevel_0], properties {delayMax=0.15, delayMin=0.15}}
 {id=Mixer_0, children [Random_3,Random_1], properties {inputVolume [1,1]}}
 {id=QualityLevel_0, children [Mixer_0,Mixer_0,Random_0,Random_0]}
 {
  id=Random_0
  children [WavePlayer_1,WavePlayer_2,WavePlayer_3,WavePlayer_4,WavePlayer_5]
  properties {bShouldExcludeFromBranchCulling=true, weights [1,1,1,1,1]}
 }
 {
  id=Random_1
  children [WavePlayer_6,WavePlayer_7,WavePlayer_8,WavePlayer_9]
  properties {bShouldExcludeFromBranchCulling=true, weights [1,1,1,1]}
 }
 {
  id=Random_3
  children [WavePlayer_14, WavePlayer_15, WavePlayer_16, WavePlayer_17, WavePlayer_18]
  properties {bShouldExcludeFromBranchCulling=true, weights [1,1,1,1,1]}
 }
 {id=WavePlayer_1, properties {soundWaveAssetPtr=&/Deimos_Fiend_HitImpact2D_05}}
 {id=WavePlayer_14, properties {soundWaveAssetPtr=&/Deimos_FiendIce_HitImpact2D_01}}
 {id=WavePlayer_15, properties {soundWaveAssetPtr=&/Deimos_FiendIce_HitImpact2D_02}}
 {id=WavePlayer_16, properties {soundWaveAssetPtr=&/Deimos_FiendIce_HitImpact2D_03}}
 {id=WavePlayer_17, properties {soundWaveAssetPtr=&/Deimos_FiendIce_HitImpact2D_04}}
 {id=WavePlayer_18, properties {soundWaveAssetPtr=&/Deimos_FiendIce_HitImpact2D_05}}
 {id=WavePlayer_2, properties {soundWaveAssetPtr=&/Deimos_Fiend_HitImpact2D_01}}
 {id=WavePlayer_3, properties {soundWaveAssetPtr=&/Deimos_Fiend_HitImpact2D_02}}
 {id=WavePlayer_4, properties {soundWaveAssetPtr=&/Deimos_Fiend_HitImpact2D_03}}
 {id=WavePlayer_5, properties {soundWaveAssetPtr=&/Deimos_Fiend_HitImpact2D_04}}
 {id=WavePlayer_6, properties {soundWaveAssetPtr=/Game/Athena/Sounds/Weapons/PickAxes/EvilCowboy/Pickaxe_EvilCowboy_Impact_01}}
 {id=WavePlayer_7, properties {soundWaveAssetPtr=/Game/Athena/Sounds/Weapons/PickAxes/EvilCowboy/Pickaxe_EvilCowboy_Impact_04}}
 {id=WavePlayer_8, properties {soundWaveAssetPtr=/Game/Athena/Sounds/Weapons/PickAxes/EvilCowboy/Pickaxe_EvilCowboy_Impact_03}}
 {id=WavePlayer_9, properties {soundWaveAssetPtr=/Game/Athena/Sounds/Weapons/PickAxes/EvilCowboy/Pickaxe_EvilCowboy_Impact_02}}
]
properties {
 maxDistance=2097152
 soundClassObject=/Game/Sounds/FortSoundClassesAndModes/Fort_Enemy_Damaged_PSM
 virtualizationMode=Disabled
 volumeMultiplier=1
 concurrencySet [/Game/Sounds/Fort_Monster_Deimos/Settings/Deimos_HItNotifications_Conc]
}
