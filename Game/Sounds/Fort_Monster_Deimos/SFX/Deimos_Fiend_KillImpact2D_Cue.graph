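# Sound cue for the 2D kill-impact sound of the Deimos Fiend monster, built from EvilCowboy pickaxe impact waves.
# A quality-level switch sends its first two slots to an equal-weight random pick of four variations, and its last two to a two-variation pick.
# It plays at reduced volume in the enemy-damaged sound class, with virtualization disabled and Deimos hit-notification concurrency limiting overlaps.
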SoundCue /Game/Sounds/Fort_Monster_Deimos/SFX/Deimos_Fiend_KillImpact2D_Cue
root=QualityLevel_2
nodes [
 {id=QualityLevel_2, children [Random_1,Random_1,Random_2,Random_2]}
 {
  id=Random_1
  children [WavePlayer_6,WavePlayer_7,WavePlayer_8,WavePlayer_9]
  properties {bShouldExcludeFromBranchCulling=true, weights [1,1,1,1]}
 }
 {id=Random_2, children [WavePlayer_0, WavePlayer_10], properties {bShouldExcludeFromBranchCulling=true, weights [1,1]}}
 {id=WavePlayer_0, properties {soundWaveAssetPtr=/Game/Athena/Sounds/Weapons/PickAxes/EvilCowboy/Pickaxe_EvilCowboy_Impact_01}}
 {id=WavePlayer_10, properties {soundWaveAssetPtr=/Game/Athena/Sounds/Weapons/PickAxes/EvilCowboy/Pickaxe_EvilCowboy_Impact_04}}
 {id=WavePlayer_6, properties {soundWaveAssetPtr=&Pickaxe_EvilCowboy_Impact_01}}
 {id=WavePlayer_7, properties {soundWaveAssetPtr=&Pickaxe_EvilCowboy_Impact_04}}
 {id=WavePlayer_8, properties {soundWaveAssetPtr=/Game/Athena/Sounds/Weapons/PickAxes/EvilCowboy/Pickaxe_EvilCowboy_Impact_03}}
 {id=WavePlayer_9, properties {soundWaveAssetPtr=/Game/Athena/Sounds/Weapons/PickAxes/EvilCowboy/Pickaxe_EvilCowboy_Impact_02}}
]
properties {
 maxDistance=2097152
 soundClassObject=/Game/Sounds/FortSoundClassesAndModes/Fort_Enemy_Damaged_PSM
 virtualizationMode=Disabled
 volumeMultiplier=0.5
 concurrencySet [/Game/Sounds/Fort_Monster_Deimos/Settings/Deimos_HItNotifications_Conc]
}
