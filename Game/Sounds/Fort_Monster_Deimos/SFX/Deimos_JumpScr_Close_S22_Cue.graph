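# Sound cue for the Season 22 close-range jump scare of the Deimos monster, routed to the enemy spawn sound class.
# Each variant mixes random jumpscare and deep-layer hits with a delayed, pitch- and volume-modulated alert from one of two random pools, each layer with its own attenuation.
# A quality node gives its first two levels the full wave pools and the last two a smaller set with added pitch variation on the hit.
# Concurrency caps it at two instances, stopping the oldest.
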SoundCue /Game/Sounds/Fort_Monster_Deimos/SFX/Deimos_JumpScr_Close_S22_Cue
root=QualityLevel_0
nodes [
 {
  id=Attenuation_0
  children [Modulator_0]
  properties {
   bOverrideAttenuation=true
   attenuationOverrides {
    falloffDistance=1024
    reverbDistanceMax=1500
    reverbDistanceMin=1
    reverbWetLevelMax=1
    reverbWetLevelMin=0.6
    stereoSpread=1024
    attenuationShapeExtents {x=1024}
   }
  }
 }
 {
  id=Attenuation_1
  children [Mixer_0]
  properties {
   bOverrideAttenuation=true
   attenuationOverrides {
    falloffDistance=1024
    reverbDistanceMax=1500
    reverbDistanceMin=1
    reverbWetLevelMax=1
    reverbWetLevelMin=0.6
    stereoSpread=1024
    attenuationShapeExtents {x=1024}
   }
  }
 }
 {
  id=Attenuation_2
  children [Random_3]
  properties {
   bOverrideAttenuation=true
   attenuationOverrides {
    falloffDistance=1024
    nonSpatializedRadiusStart=600
    reverbDistanceMax=1500
    reverbDistanceMin=1
    reverbWetLevelMax=0.6
    reverbWetLevelMin=0.2
    stereoSpread=20
    attenuationShapeExtents {x=1024}
   }
  }
 }
 {
  id=Attenuation_3
  children [Random_9]
  properties {
   bOverrideAttenuation=true
   attenuationOverrides {
    falloffDistance=1024
    nonSpatializedRadiusStart=600
    reverbDistanceMax=1500
    reverbDistanceMin=1
    reverbWetLevelMax=0.6
    reverbWetLevelMin=0.2
    stereoSpread=20
    attenuationShapeExtents {x=1024}
   }
  }
 }
 {id=Delay_0, children [Random_0], properties {delayMax=1, delayMin=0.5}}
 {id=Delay_2, children [Random_1], properties {delayMax=0.5, delayMin=0.5}}
 {id=Delay_3, children [Random_7], properties {delayMax=1.2, delayMin=0.7}}
 {id=Delay_4, children [Random_8], properties {delayMax=0.5, delayMin=0.5}}
 {id=Mixer_0, children [Random_4,Random_5], properties {inputVolume [1.25,1]}}
 {id=Mixer_1, children [Random_6, WavePlayer_52], properties {inputVolume [1.25,1]}}
 {id=Mixer_2, children [Attenuation_1, Modulator_1], properties {inputVolume [0.9,1.35]}}
 {id=Mixer_3, children [Attenuation_0, Modulator_3], properties {inputVolume [0.9,1.35]}}
 {id=Modulator_0, children [Mixer_1], properties {pitchMax=1.15, pitchMin=0.85}}
 {id=Modulator_1, children [Attenuation_2], properties {pitchMax=1.1, pitchMin=0.9, volumeMax=1.1, volumeMin=0.9}}
 {id=Modulator_3, children [Attenuation_3], properties {pitchMax=1.1, pitchMin=0.9, volumeMax=1.1, volumeMin=0.9}}
 {id=QualityLevel_0, children [SoundClass_0,SoundClass_0,SoundClass_1,SoundClass_1]}
 {id=Random_0, children [WavePlayer_11, WavePlayer_19, WavePlayer_23, WavePlayer_24, WavePlayer_25], properties {weights [1,1,1,1,1]}}
 {id=Random_1, children [WavePlayer_26, WavePlayer_32, WavePlayer_33, WavePlayer_36, WavePlayer_37], properties {weights [1,1,1,1,1]}}
 {id=Random_3, children [Delay_0,Delay_2], properties {weights [1,1]}}
 {
  id=Random_4
  children [
   WavePlayer_0
   WavePlayer_12
   WavePlayer_13
   WavePlayer_14
   WavePlayer_15
   WavePlayer_16
   WavePlayer_17
   WavePlayer_18
   WavePlayer_20
   WavePlayer_21
   WavePlayer_22
   WavePlayer_27
  ]
  properties {weights [1,1,1,1,1,1,1,1,1,1,1,1]}
 }
 {
  id=Random_5
  children [WavePlayer_28, WavePlayer_29, WavePlayer_30, WavePlayer_31, WavePlayer_34, WavePlayer_35]
  properties {weights [1,1,1,1,1,1]}
 }
 {id=Random_6, children [WavePlayer_40, WavePlayer_41, WavePlayer_45], properties {weights [1,1,1]}}
 {id=Random_7, children [WavePlayer_48, WavePlayer_49], properties {weights [1,1]}}
 {id=Random_8, children [WavePlayer_53, WavePlayer_56], properties {weights [1,1]}}
 {id=Random_9, children [Delay_3,Delay_4], properties {weights [1,1]}}
 {id=SoundClass_0, children [Mixer_2], properties {soundClassOverride=/Game/Sounds/FortSoundClassesAndModes/Fort_Enemy_Spawn}}
 {id=SoundClass_1, children [Mixer_3], properties {soundClassOverride=&Fort_Enemy_Spawn}}
 {id=WavePlayer_0, properties {soundWaveAssetPtr=&/Jumpscare_S22_01}}
 {id=WavePlayer_11, properties {soundWaveAssetPtr=&/Jumpscare_S22_Alert_01}}
 {id=WavePlayer_12, properties {soundWaveAssetPtr=&/Jumpscare_S22_02}}
 {id=WavePlayer_13, properties {soundWaveAssetPtr=&/Jumpscare_S22_03}}
 {id=WavePlayer_14, properties {soundWaveAssetPtr=&/Jumpscare_S22_04}}
 {id=WavePlayer_15, properties {soundWaveAssetPtr=&/Jumpscare_S22_05}}
 {id=WavePlayer_16, properties {soundWaveAssetPtr=&/Jumpscare_S22_06}}
 {id=WavePlayer_17, properties {soundWaveAssetPtr=&/Jumpscare_S22_07}}
 {id=WavePlayer_18, properties {soundWaveAssetPtr=&/Jumpscare_S22_08}}
 {id=WavePlayer_19, properties {soundWaveAssetPtr=&/Jumpscare_S22_Alert_02}}
 {id=WavePlayer_20, properties {soundWaveAssetPtr=&/Jumpscare_S22_10}}
 {id=WavePlayer_21, properties {soundWaveAssetPtr=&/Jumpscare_S22_11}}
 {id=WavePlayer_22, properties {soundWaveAssetPtr=&/Jumpscare_S22_12}}
 {id=WavePlayer_23, properties {soundWaveAssetPtr=&/Jumpscare_S22_Alert_03}}
 {id=WavePlayer_24, properties {soundWaveAssetPtr=&/Jumpscare_S22_Alert_04}}
 {id=WavePlayer_25, properties {soundWaveAssetPtr=&/Jumpscare_S22_Alert_05}}
 {id=WavePlayer_26, properties {soundWaveAssetPtr=&/Jumpscare_S22_Alert_B_01}}
 {id=WavePlayer_27, properties {soundWaveAssetPtr=&/Jumpscare_S22_17}}
 {id=WavePlayer_28, properties {soundWaveAssetPtr=&/Jumpscare_S22_DeepLayer_01}}
 {id=WavePlayer_29, properties {soundWaveAssetPtr=&/Jumpscare_S22_DeepLayer_02}}
 {id=WavePlayer_30, properties {soundWaveAssetPtr=&/Jumpscare_S22_DeepLayer_03}}
 {id=WavePlayer_31, properties {soundWaveAssetPtr=&/Jumpscare_S22_DeepLayer_04}}
 {id=WavePlayer_32, properties {soundWaveAssetPtr=&/Jumpscare_S22_Alert_B_02}}
 {id=WavePlayer_33, properties {soundWaveAssetPtr=&/Jumpscare_S22_Alert_B_03}}
 {id=WavePlayer_34, properties {soundWaveAssetPtr=&/Jumpscare_S22_DeepLayer_07}}
 {id=WavePlayer_35, properties {soundWaveAssetPtr=&/Jumpscare_S22_DeepLayer_08}}
 {id=WavePlayer_36, properties {soundWaveAssetPtr=&/Jumpscare_S22_Alert_B_04}}
 {id=WavePlayer_37, properties {soundWaveAssetPtr=&/Jumpscare_S22_Alert_B_05}}
 {id=WavePlayer_40, properties {soundWaveAssetPtr=&Jumpscare_S22_01}}
 {id=WavePlayer_41, properties {soundWaveAssetPtr=&Jumpscare_S22_02}}
 {id=WavePlayer_45, properties {soundWaveAssetPtr=&Jumpscare_S22_06}}
 {id=WavePlayer_48, properties {soundWaveAssetPtr=&Jumpscare_S22_Alert_02}}
 {id=WavePlayer_49, properties {soundWaveAssetPtr=&Jumpscare_S22_Alert_03}}
 {id=WavePlayer_52, properties {soundWaveAssetPtr=&Jumpscare_S22_DeepLayer_01}}
 {id=WavePlayer_53, properties {soundWaveAssetPtr=&Jumpscare_S22_Alert_B_01}}
 {id=WavePlayer_56, properties {soundWaveAssetPtr=&Jumpscare_S22_Alert_B_04}}
]
properties {
 bOverrideConcurrency=true
 maxDistance=2048
 virtualizationMode=Disabled
 volumeMultiplier=1.8
 attenuationOverrides {bSpatialize=false, attenuationShapeExtents {x=1800}}
 concurrencyOverrides {maxCount=2, resolutionRule=StopOldest, volumeScaleAttackTime=0.2}
}
