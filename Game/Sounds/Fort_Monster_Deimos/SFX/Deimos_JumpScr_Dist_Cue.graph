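# Sound cue for a distant Deimos monster sound effect, named as a jump scare, on the enemy spawn sound class.
# A single wave player plays the Deimos spawner explosion wave through a modulator that fixes it at a raised pitch.
# It uses the large spawner ambient attenuation asset, and a concurrency override allows one instance at a time, blocking new plays while it sounds.
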SoundCue /Game/Sounds/Fort_Monster_Deimos/SFX/Deimos_JumpScr_Dist_Cue
root=Modulator_0
nodes [
 {id=Modulator_0, children [WavePlayer_12], properties {pitchMax=1.5, pitchMin=1.5}}
 {id=WavePlayer_12, properties {soundWaveAssetPtr=/Game/Sounds/Fort_Monster_Deimos/Spawner/DeimosSpawner_Explode}}
]
properties {
 attenuationSettings=/Game/Sounds/Fort_Monster_Deimos/Spawner/Fort_Deimos_SpawnerAmbientLarge_Att
 bOverrideConcurrency=true
 maxDistance=4000
 soundClassObject=/Game/Sounds/FortSoundClassesAndModes/Fort_Enemy_Spawn
 volumeMultiplier=1.4
 attenuationOverrides {bSpatialize=false, attenuationShapeExtents {x=1800}}
 concurrencyOverrides {maxCount=1, resolutionRule=PreventNew}
}
