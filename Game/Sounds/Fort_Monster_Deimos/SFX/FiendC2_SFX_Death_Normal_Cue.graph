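# Sound cue for the normal death of a Chapter 2 Fiend, a Deimos monster, attenuated with the Deimos death explosion settings.
# A mixer layers two randomized branches: one of five FiendCP death effects with slight upward pitch variation, and one of seven newer Fiend death sounds, pitch-modulated and delayed by a fixed short offset.
# It plays under the enemy-damaged sound class with shared Deimos concurrency and virtualization disabled.
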SoundCue /Game/Sounds/Fort_Monster_Deimos/SFX/FiendC2_SFX_Death_Normal_Cue
root=Mixer_0
nodes [
 {id=Delay_0, children [Modulator_0], properties {delayMax=0.2, delayMin=0.2}}
 {id=Mixer_0, children [Modulator_1,Delay_0], properties {inputVolume [2.2,2]}}
 {id=Modulator_0, children [Random_3], properties {pitchMax=1.2}}
 {id=Modulator_1, children [Random_0], properties {pitchMax=1.1, pitchMin=1}}
 {
  id=Random_0
  children [WavePlayer_0,WavePlayer_1,WavePlayer_2,WavePlayer_3,WavePlayer_4]
  properties {bShouldExcludeFromBranchCulling=true, weights [1,1,1,1,1]}
 }
 {
  id=Random_3
  children [WavePlayer_15, WavePlayer_16, WavePlayer_18, WavePlayer_19, WavePlayer_20, WavePlayer_21, WavePlayer_22]
  properties {bShouldExcludeFromBranchCulling=true, weights [1,1,1,1,1,1,1]}
 }
 {id=WavePlayer_0, properties {soundWaveAssetPtr=&/FiendCP_DeathSFX_04}}
 {id=WavePlayer_1, properties {soundWaveAssetPtr=&/FiendCP_DeathSFX_01}}
 {id=WavePlayer_15, properties {soundWaveAssetPtr=/Game/Sounds/Fort_Monster_Deimos/Chapter2/Fiend_Death_New_01}}
 {id=WavePlayer_16, properties {soundWaveAssetPtr=/Game/Sounds/Fort_Monster_Deimos/Chapter2/Fiend_Death_New_02}}
 {id=WavePlayer_18, properties {soundWaveAssetPtr=/Game/Sounds/Fort_Monster_Deimos/Chapter2/Fiend_Death_New_03}}
 {id=WavePlayer_19, properties {soundWaveAssetPtr=/Game/Sounds/Fort_Monster_Deimos/Chapter2/Fiend_Death_New_04}}
 {id=WavePlayer_2, properties {soundWaveAssetPtr=&/FiendCP_DeathSFX_02}}
 {id=WavePlayer_20, properties {soundWaveAssetPtr=/Game/Sounds/Fort_Monster_Deimos/Chapter2/Fiend_Death_New_05}}
 {id=WavePlayer_21, properties {soundWaveAssetPtr=/Game/Sounds/Fort_Monster_Deimos/Chapter2/Fiend_Death_New_06}}
 {id=WavePlayer_22, properties {soundWaveAssetPtr=/Game/Sounds/Fort_Monster_Deimos/Chapter2/Fiend_Death_New_07}}
 {id=WavePlayer_3, properties {soundWaveAssetPtr=&/FiendCP_DeathSFX_03}}
 {id=WavePlayer_4, properties {soundWaveAssetPtr=&/FiendCP_DeathSFX_05}}
]
properties {
 attenuationSettings=&/Deimos_Death_Explo_Att
 maxDistance=13000
 soundClassObject=/Game/Sounds/FortSoundClassesAndModes/Fort_Enemy_Damaged_PSM
 virtualizationMode=Disabled
 volumeMultiplier=0.8
 concurrencyOverrides {maxCount=2}
 concurrencySet [/Game/Sounds/Fort_Monster_Deimos/Settings/Deimos_Global_Conc]
}
