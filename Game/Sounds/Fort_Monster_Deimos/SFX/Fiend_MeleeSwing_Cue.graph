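# Sound cue for the Deimos Fiend monster's melee swing attack.
# A Random node picks one of six equally weighted swing waves and feeds a Modulator that varies pitch on each play.
# It plays in the creatures sound class with a melee attack attenuation asset, global and local Deimos attack concurrency sets, boosted volume and virtualization disabled.
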SoundCue /Game/Sounds/Fort_Monster_Deimos/SFX/Fiend_MeleeSwing_Cue
root=Modulator_1
nodes [
 {id=Modulator_1, children [Random_0], properties {pitchMax=1.2, pitchMin=0.9}}
 {
  id=Random_0
  children [WavePlayer_0, WavePlayer_1, WavePlayer_2, WavePlayer_3, WavePlayer_4, WavePlayer_12]
  properties {bShouldExcludeFromBranchCulling=true, weights [1,1,1,1,1,1]}
 }
 {id=WavePlayer_0, properties {soundWaveAssetPtr=&/Deimos_Fiend_Swing_01}}
 {id=WavePlayer_1, properties {soundWaveAssetPtr=&/Deimos_Fiend_Swing_02}}
 {id=WavePlayer_12, properties {soundWaveAssetPtr=&/Deimos_Fiend_Swing_06}}
 {id=WavePlayer_2, properties {soundWaveAssetPtr=&/Deimos_Fiend_Swing_03}}
 {id=WavePlayer_3, properties {soundWaveAssetPtr=&/Deimos_Fiend_Swing_04}}
 {id=WavePlayer_4, properties {soundWaveAssetPtr=&/Deimos_Fiend_Swing_05}}
]
properties {
 attenuationSettings=/Game/Sounds/Attenuation/Fort_Deimos_Attack_Melee_Att
 bDebug=true
 maxDistance=3300
 soundClassObject=/Game/Sounds/FortSoundClassesAndModes/Fort_Creatures
 virtualizationMode=Disabled
 volumeMultiplier=3.5655
 attenuationOverrides {distanceAlgorithm=Logarithmic}
 concurrencyOverrides {maxCount=3}
 concurrencySet [/Game/Sounds/Fort_Monster_Deimos/Settings/Deimos_Attack_Global_Conc, /Game/Sounds/Fort_Monster_Deimos/Settings/Deimos_Attack_Local_Conc]
}
