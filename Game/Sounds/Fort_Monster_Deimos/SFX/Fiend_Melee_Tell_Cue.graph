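# Sound cue for the Fiend monster's melee attack tell in the Deimos creature set.
# A quality-level switch sends the first level to a random pick of four tell waves and the other three levels to a shared random pick of two, each through a modulator.
# It plays in the creatures sound class with Deimos melee attenuation, global and local attack concurrency sets, and virtualization disabled.
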SoundCue /Game/Sounds/Fort_Monster_Deimos/SFX/Fiend_Melee_Tell_Cue
root=QualityLevel_0
nodes [
 {id=Modulator_1, children [Random_0]}
 {id=Modulator_2, children [Random_2]}
 {id=QualityLevel_0, children [Modulator_1,Modulator_2,Modulator_2,Modulator_2]}
 {id=Random_0, children [WavePlayer_0,WavePlayer_1,WavePlayer_2,WavePlayer_3], properties {weights [1,1,1,1]}}
 {id=Random_2, children [WavePlayer_11, WavePlayer_12], properties {weights [1,1]}}
 {id=WavePlayer_0, properties {soundWaveAssetPtr=&/Fiend_Melee_Tell_01}}
 {id=WavePlayer_1, properties {soundWaveAssetPtr=&/Fiend_Melee_Tell_02}}
 {id=WavePlayer_11, properties {soundWaveAssetPtr=&Fiend_Melee_Tell_01}}
 {id=WavePlayer_12, properties {soundWaveAssetPtr=&Fiend_Melee_Tell_02}}
 {id=WavePlayer_2, properties {soundWaveAssetPtr=&/Fiend_Melee_Tell_03}}
 {id=WavePlayer_3, properties {soundWaveAssetPtr=&/Fiend_Melee_Tell_04}}
]
properties {
 attenuationSettings=/Game/Sounds/Attenuation/Fort_Deimos_Attack_Melee_Att
 maxDistance=3300
 soundClassObject=/Game/Sounds/FortSoundClassesAndModes/Fort_Creatures
 virtualizationMode=Disabled
 volumeMultiplier=2.3775
 attenuationOverrides {distanceAlgorithm=Logarithmic}
 concurrencyOverrides {maxCount=2}
 concurrencySet [/Game/Sounds/Fort_Monster_Deimos/Settings/Deimos_Attack_Global_Conc, /Game/Sounds/Fort_Monster_Deimos/Settings/Deimos_Attack_Local_Conc]
}
