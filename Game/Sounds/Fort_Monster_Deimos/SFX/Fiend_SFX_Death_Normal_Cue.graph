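# Sound cue for the normal death SFX of the Fiend, a Deimos monster, using close vocal attenuation and the enemy-damaged sound class.
# A quality-level root sends its first tier to a mixer whose only connected input is a boosted, briefly delayed, pitch-modulated random pick of Fiend death waves.
# The second tier instead plays pitch-modulated, randomized Husk death zaps, and the remaining tiers are empty.
# Shared Deimos concurrency limits simultaneous playback, and virtualization is disabled.
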SoundCue /Game/Sounds/Fort_Monster_Deimos/SFX/Fiend_SFX_Death_Normal_Cue
root=QualityLevel_1
nodes [
 {id=Delay_0, children [Modulator_0], properties {delayMax=0.2, delayMin=0.2}}
 {id=Mixer_0, children [null,Delay_0], properties {inputVolume [0.4,2]}}
 {id=Modulator_0, children [Random_0], properties {pitchMax=1.2}}
 {id=Modulator_2, children [Random_2], properties {pitchMax=1.4, pitchMin=1}}
 {id=QualityLevel_1, children [Mixer_0,Modulator_2,null,null]}
 {
  id=Random_0
  children [WavePlayer_4, WavePlayer_5, WavePlayer_0, WavePlayer_1, WavePlayer_2, WavePlayer_3, WavePlayer_14]
  properties {weights [1,1,1,1,1,1,1]}
 }
 {id=Random_2, children [WavePlayer_12, WavePlayer_13, WavePlayer_17], properties {weights [1,1,1]}}
 {id=WavePlayer_0, properties {soundWaveAssetPtr=/Game/Sounds/Fort_Monster_Deimos/Fiend/Fiend_Death_New_03}}
 {id=WavePlayer_1, properties {soundWaveAssetPtr=/Game/Sounds/Fort_Monster_Deimos/Fiend/Fiend_Death_New_04}}
 {id=WavePlayer_12, properties {soundWaveAssetPtr=/Game/Sounds/Fort_Monster_Husk/SFX/husk_death_zap_D_01}}
 {id=WavePlayer_13, properties {soundWaveAssetPtr=/Game/Sounds/Fort_Monster_Husk/SFX/husk_death_zap_D_02}}
 {id=WavePlayer_14, properties {soundWaveAssetPtr=/Game/Sounds/Fort_Monster_Deimos/Fiend/Fiend_Death_New_07}}
 {id=WavePlayer_17, properties {soundWaveAssetPtr=/Game/Sounds/Fort_Monster_Husk/SFX/husk_death_zap_D_06}}
 {id=WavePlayer_2, properties {soundWaveAssetPtr=/Game/Sounds/Fort_Monster_Deimos/Fiend/Fiend_Death_New_05}}
 {id=WavePlayer_3, properties {soundWaveAssetPtr=/Game/Sounds/Fort_Monster_Deimos/Fiend/Fiend_Death_New_06}}
 {id=WavePlayer_4, properties {soundWaveAssetPtr=/Game/Sounds/Fort_Monster_Deimos/Fiend/Fiend_Death_New_01}}
 {id=WavePlayer_5, properties {soundWaveAssetPtr=/Game/Sounds/Fort_Monster_Deimos/Fiend/Fiend_Death_New_02}}
]
properties {
 attenuationSettings=/Game/Sounds/Attenuation/Fort_Deimos_Vocal_Close_Att
 maxDistance=13000
 soundClassObject=/Game/Sounds/FortSoundClassesAndModes/Fort_Enemy_Damaged_PSM
 virtualizationMode=Disabled
 volumeMultiplier=0.35
 concurrencyOverrides {maxCount=2}
 concurrencySet [/Game/Sounds/Fort_Monster_Deimos/Settings/Deimos_Global_Conc]
}
